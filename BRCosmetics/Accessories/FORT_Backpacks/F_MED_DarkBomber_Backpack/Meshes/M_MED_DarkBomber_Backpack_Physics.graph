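# Physics asset for the DarkBomber backpack cosmetic, letting a hanging chain swing against the pack body.
# The anim bone carries a box and a capsule collision shape for the backpack itself.
# Five small capsule bodies, dyn_chain_01 to 05, link serially from anim through constraints sharing identical swing and tight twist limits.
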PhysicsAsset /BRCosmetics/Accessories/FORT_Backpacks/F_MED_DarkBomber_Backpack/Meshes/M_MED_DarkBomber_Backpack_Physics
bodies {
 anim [
  {shapeType=Capsule, length=8.8442, radius=3.5777, center=<16.6301 1.7882 -8.3662>, rotation=<rot 0.0002 24.9956 89.9995>}
  {shapeType=Box, extentX=13.3845, extentY=13.5328, extentZ=24.107, center=<8.678 -1.27 0>, rotation=<rot 0 24.9992 0>}
 ]
 dyn_chain_01 [{shapeType=Capsule, length=1, radius=0.9958, center=<0 0 -1>}]
 dyn_chain_02 [{shapeType=Capsule, length=1, radius=0.9958, center=<0 0 -1>}]
 dyn_chain_03 [{shapeType=Capsule, length=1, radius=0.9958, center=<0 0 -1>}]
 dyn_chain_04 [{shapeType=Capsule, length=1, radius=0.9958, center=<0 0 -1>}]
 dyn_chain_05 [{shapeType=Capsule, length=1, radius=0.9958, center=<0 0 -1>}]
}
constraints [
 {bone1Name=dyn_chain_01, bone2Name=anim, swing1LimitDegrees=60, swing2LimitDegrees=60, twistLimitDegrees=10}
 {bone1Name=dyn_chain_02, bone2Name=dyn_chain_01, swing1LimitDegrees=60, swing2LimitDegrees=60, twistLimitDegrees=10}
 {bone1Name=dyn_chain_03, bone2Name=dyn_chain_02, swing1LimitDegrees=60, swing2LimitDegrees=60, twistLimitDegrees=10}
 {bone1Name=dyn_chain_04, bone2Name=dyn_chain_03, swing1LimitDegrees=60, swing2LimitDegrees=60, twistLimitDegrees=10}
 {bone1Name=dyn_chain_05, bone2Name=dyn_chain_04, swing1LimitDegrees=60, swing2LimitDegrees=60, twistLimitDegrees=10}
]
